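# Material function for the LocalZilla backpack that turns vertex colors into neon-sign emissive and base color tints within the incoming material attributes.
# Red bands select five neon colors, with alpha driving a timed two-group blink on two of them, and blue dims the glow.
# Time of day and shading path adjust glow, computed per vertex through a VertexInterpolator before masking by incoming emissive.
# Green chooses tube darkening or two junction tints for base color.
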MaterialFunction /BRCosmetics/Accessories/FORT_Backpacks/Backpack_M_MED_LocalZilla/Materials/MF_MED_LocalZilla_Backpack_VertexColorsToEmissive
graph {nodes [
  n1 = FunctionInput() {inputType=FunctionInput_MaterialAttributes}
  n2 = GetMaterialAttributes(None=@n1) {attributeGetTypes [#1, #2]}
  n3 = Saturate(None=@n2.EmissiveColor)
  n4 = VertexColor()
  n5 = Add(A=@n4.R) {constB=0.05}
  n6 = Floor(None=Add(A=@n5) {constB=0.3})
  n7 = Subtract(A=Floor(None=Add(A=@n5) {constB=0.4}), B=@n6)
  n8 = VectorParameter() {group=Neon Colors, parameterName=NeonColor5, sortPriority=0, defaultValue=<rgba 1 1 1 1>}
  n9 = Floor(None=Add(A=@n5) {constB=0.2})
  n10 = Subtract(A=@n6, B=@n9)
  n11 = VectorParameter() {group=Neon Colors, parameterName=NeonColor4, sortPriority=0, defaultValue=<rgba 1 1 1 1>}
  n12 = Floor(None=Add(A=@n5) {constB=0.1})
  n13 = Multiply(
   A=Multiply(
    A=Subtract(A=@n9, B=@n12)
    B=VectorParameter() {group=Neon Colors, parameterName=NeonColor3, sortPriority=0, defaultValue=<rgba 1 1 1 1>}.RGB
   )
   B=NamedRerouteUsage()
  )
  n14 = Floor(None=@n5)
  n15 = Add(
   A=Multiply(
    A=Subtract(A=@n12, B=@n14)
    B=VectorParameter() {group=Neon Colors, parameterName=NeonColor2, sortPriority=0, defaultValue=<rgba 1 1 1 1>}.RGB
   )
   B=Multiply(A=@n14, B=VectorParameter() {group=Neon Colors, parameterName=NeonColor1, sortPriority=0, defaultValue=<rgba 1 1 1 1>}.RGB)
  )
  n16 = Add(A=Multiply(A=Multiply(A=@n7, B=@n8.RGB), B=NamedRerouteUsage()), B=Add(A=Multiply(A=@n10, B=@n11.RGB), B=Add(A=@n13, B=@n15)))
  n17 = Constant()
  n18 = MaterialFunctionCall(
   Value_Morning=@n17
   Value_Day=@n17
   Value_Evening=@n17
   Value_Night=ScalarParameter() {defaultValue=0.1, group=Neon Colors, parameterName=NightDesaturationFraction, sortPriority=3}
  ) {materialFunction=&Scalar_Time_of_Day}
  n19 = ScalarParameter() {defaultValue=1, group=Emissive, parameterName=EmissiveFactor_Deferred}
  n20 = ScalarParameter() {defaultValue=1.5, group=Emissive, parameterName=EmissiveFactor_Mobile/Forward}
  n21 = ScalarParameter() {defaultValue=0.75, group=Neon Colors, parameterName=MorningEveningIntensityFactor, sortPriority=1}
  n22 = Multiply(
   A=Multiply(
    A=ScalarParameter() {defaultValue=1, group=Emissive, parameterName=emissive mult}
    B=ShadingPathSwitch(Default=@n19, Deferred=@n19, Forward=@n20, Mobile=@n20)
   )
   B=MaterialFunctionCall(
    Value_Morning=@n21
    Value_Day=Constant() {r=1}
    Value_Evening=@n21
    Value_Night=ScalarParameter() {defaultValue=0.5, group=Neon Colors, parameterName=NightIntensityFactor, sortPriority=2}
   ) {materialFunction=&Scalar_Time_of_Day}.Result
  )
  n23 = VertexColor()
  n24 = OneMinus(None=@n23.B)
  n25 = VertexInterpolator(VS=Multiply(A=Desaturation(None=@n16, Fraction=@n18.Result), B=Multiply(A=@n22, B=@n24)))
  n26 = Multiply(A=@n2.BaseColor, B=ScalarParameter() {defaultValue=0.5, group=Neon Colors, parameterName=TubeBaseColorBrightness})
  n27 = LinearInterpolate(
   A=@n26
   B=Multiply(A=@n2.BaseColor, B=VectorParameter() {group=Neon Colors, parameterName=JunctionColor1, sortPriority=1, defaultValue=<rgba 1 1 1 1>}.RGB)
   Alpha=Floor(None=Add(A=@n23.G) {constB=0.0001})
  )
  n28 = Multiply(A=@n2.BaseColor, B=VectorParameter() {group=Neon Colors, parameterName=JunctionColor2, sortPriority=1, defaultValue=<rgba 1 1 1 1>}.RGB)
  FunctionOutput(None=SetMaterialAttributes!(
    MaterialAttributes=@n1
    EmissiveColor=Multiply(A=@n3, B=@n25.PS) {constB=4}
    BaseColor=If(A=@n23.G, "A > B"=@n27, "A == B"=@n28, "A < B"=@n26) {constB=0.5, equalsThreshold=0.002}
   ) {attributeSetTypes [#1, #2]})
  n29 = Ceil(None=Subtract(A=@n4.A) {constB=0.6})
  n30 = OneMinus(None=@n29)
  n31 = Ceil(None=@n4.A)
  n32 = Frac(None=Multiply(A=Time()) {constB=0.5})
  NamedRerouteDeclaration(None=LinearInterpolate(A=@n29, B=Multiply(A=@n30, B=@n31), Alpha=Round(None=@n32))) {name=Blinking, nodeColor=<rgba 0 1 0.0625 1>}
 ]}
