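# Additive Niagara mesh-particle material for the neon glow FX on the LocalZilla backpack, colored by the backpack's vertex-color-to-emissive function plus an emissive adjustment.
# Floor thresholds split vertex red into bands, some steady and others toggling over time between two vertex-alpha groups, like an animated sign.
# That mask drives opacity, multiplied by particle alpha, a soft U-axis falloff peaking mid-UV, and a view-angle fade.
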Material /BRCosmetics/Accessories/FORT_Backpacks/Backpack_M_MED_LocalZilla/FX/M_LocalZilla_Backpack_NeonGlow
graph {
 properties {
  bUseMaterialAttributes=true
  bUsedWithNiagaraMeshParticles=true
  blendMode=BLEND_Additive
  pixelDepthOffsetMode=PDOM_Legacy
  usedShadingModels=Default Lit
 }
 nodes [
  n1 = Constant() {r=1}
  n2 = MaterialFunctionCall!(In=MakeMaterialAttributes!(
    BaseColor=@n1
    Metallic
    Specular
    Roughness
    Anisotropy
    EmissiveColor=@n1
    Opacity
    OpacityMask
    Normal
    Tangent
    WorldPositionOffset
    SubsurfaceColor
    ClearCoat
    ClearCoatRoughness
    AmbientOcclusion
    Refraction
    CustomizedUVs_0
    CustomizedUVs_1
    CustomizedUVs_2
    CustomizedUVs_3
    CustomizedUVs_4
    CustomizedUVs_5
    CustomizedUVs_6
    CustomizedUVs_7
    PixelDepthOffset
    ShadingModel
    Displacement
   )) {materialFunction=/BRCosmetics/Accessories/FORT_Backpacks/Backpack_M_MED_LocalZilla/Materials/MF_MED_LocalZilla_Backpack_VertexColorsToEmissive}
  n3 = BreakMaterialAttributes!(Attr=@n2.Result)
  n4 = ComponentMask(None=TextureCoordinate()) {r=true}
  n5 = Multiply(A=@n4, B=OneMinus(None=@n4))
  n6 = Multiply(A=ParticleColor().A, B=Power(Base=Multiply(A=@n5) {constB=5}) {constExponent=3})
  n7 = VertexColor()
  n8 = Add(A=@n7.R) {constB=0.05}
  n9 = Floor(None=Add(A=@n8) {constB=0.4})
  n10 = Floor(None=Add(A=@n8) {constB=0.3})
  n11 = Ceil(None=Subtract(A=@n7.A) {constB=0.6})
  n12 = Multiply(A=Time()) {constB=0.5}
  n13 = LinearInterpolate(A=@n11, B=Multiply(A=OneMinus(None=@n11), B=Ceil(None=@n7.A)), Alpha=Round(None=Frac(None=@n12)))
  n14 = Floor(None=Add(A=@n8) {constB=0.2})
  n15 = Floor(None=Add(A=@n8) {constB=0.1})
  n16 = Multiply(A=Subtract(A=@n14, B=@n15), B=@n13)
  n17 = Floor(None=@n8)
  n18 = Add(A=Subtract(A=@n15, B=@n17), B=@n17)
  n19 = Add(A=Multiply(A=Subtract(A=@n9, B=@n10), B=@n13), B=Add(A=Subtract(A=@n10, B=@n14), B=Add(A=@n16, B=@n18)))
  n20 = Constant() {r=2}
  MakeMaterialAttributes(
   BaseColor=@n3.BaseColor
   EmissiveColor=MaterialFunctionCall(Emissive=@n3.EmissiveColor) {materialFunction=&MF_FX_EmissiveAdjustments}.Result
   Opacity=Multiply(
    A=Multiply(A=@n6, B=@n19)
    B=MaterialFunctionCall("FadeSensitivity (optional)"=@n20) {materialFunction=&FadeBasedOnViewAngle}.Result
   )
  )
 ]
}
