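# Niagara system for the LocalZilla backpack cosmetic, adding glowing sprites, meshes and lights that follow its skeletal mesh.
# Two burst emitters keep looping particles attached: a blink light with an offset alternating on engine time, and a base glow with mesh renderers.
# System-level colors tint the lights, and curves make them flicker.
# User.Emissive scales brightness and alpha, while front-end flags disable the particle lights.
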NiagaraSystem /BRCosmetics/Accessories/FORT_Backpacks/Backpack_M_MED_LocalZilla/FX/NS_Backpack_LocalZilla
dependencies {
 usedDataInterfaces [&NiagaraDataInterfaceCurve, &NiagaraDataInterfaceArrayFloat3]
 usedDynamicInputs [&SetBoolByFloatComparison, &MakeCustomLinearColorFromBool]
 usedModules [
  &SystemState
  &self:SystemSpawnScript.SystemScriptSource.SystemScriptGraph.NiagaraNodeAssignment_2.SetVariables_#1
  &self:SystemSpawnScript.SystemScriptSource.SystemScriptGraph.NiagaraNodeAssignment_0.SetVariables_#2
 ]
}
emitters {
 DirectionalBurst {
  emitterUpdateScript [
   &EmitterState()
   &SpawnBurst_Instantaneous(Use Spawn Probability=false, UseLoopCountLimit=true)
  ]
  particleSpawnScript [&InitializeParticle(
    Color Mode=::NewEnumerator0
    Lifetime=6
    Sprite Size Mode=::NewEnumerator3
    Uniform Sprite Size=12
    UsePositionOffset=false
    Write Material Random=true
   )]
  particleUpdateScript [
   &ParticleState(
    Kill Particles When Lifetime Has Elapsed=false
    Let Infinitely Lived Particles Die When Emitter Deactivates=true
    Loop Particles Lifetime=true
   )
   &Color(Color=$System.BlinkLigthColor)
   &SkeletalMeshLocation(
    MaskSpawnGroup=false
    Mesh Sampling Type=::NewEnumerator5
    OverrideRandomSeed=false
    Sampled Position Offset=<0 0.1 0>
    Skeletal Mesh=$User.SkeletalMesh
   ) {module=SkeletalMeshLocation001}
   /SetVariables(Particles.Blink:NiagaraFloat=&FloatFromCurve(FloatCurve=DataInterface() {propertyValues {Curve {keys [
         {arriveTangent=-1.6748, interpMode=RCIM_Cubic, time=0, value=1}
         {arriveTangent=-0.0227, interpMode=RCIM_Cubic, leaveTangent=-0.0227, tangentMode=RCTM_User, time=0.1526, value=1.25}
         {arriveTangent=-0.209, interpMode=RCIM_Cubic, leaveTangent=-0.209, time=0.2819, value=0.9}
         {arriveTangent=-0.4585, interpMode=RCIM_Cubic, leaveTangent=-0.4585, time=0.3918, value=1.2}
         {arriveTangent=-1.026, interpMode=RCIM_Cubic, leaveTangent=-1.026, time=0.5, value=0.8}
         {arriveTangent=2.311, interpMode=RCIM_Cubic, leaveTangent=2.311, time=0.6112, value=0.9748}
         {arriveTangent=-0.4762, interpMode=RCIM_Cubic, leaveTangent=-0.4762, time=0.738, value=1.35}
         {arriveTangent=-1.3358, interpMode=RCIM_Cubic, leaveTangent=-1.3358, time=0.8734, value=0.85}
         <at 1 1 RCIM_Cubic>
        ]}}})) {isSetParameters=true}
   &FN_Light_Attributes(
    Color=$Particles.Color
    Create Color Attribute=true
    Create Exponent Attribute=true
    Create Radius Attribute=true
    Enable Particle Light Enable=true
    Exponent=3
    Light Enabled:NiagaraBool=&InvertBool(Boolean:NiagaraBool=&Boolean_OrOperation(Bool1=$User.bIsFrontEnd, Bool2=$User.bIsFrontEndPreview))
    Radius=30
    Scale Color=&Multiply_Float(A=$Particles.Blink, B=&ClampFloat(Float=$User.Emissive))
   )
   /SetVariables(Particles.TimeBlink:NiagaraFloat=/Niagara/DynamicInputs/Integer/V2/Round_Float(Float:NiagaraFloat=&Frac_Float(Float=&Multiply_Float(A=$Engine.Time, B=0.5)))) {isSetParameters=true}
   /SetVariables(Particles.PositionLight:NiagaraPosition=&Add_Vector(
     A=Linked() {linkedVariable {name=Particles.Position, type=NiagaraPosition}}
     B=&SelectVectorFromArray(
      Direct Array Index=&MakeIntFromFloat(Float=$Particles.TimeBlink)
      Direct Array Mode=::NewEnumerator1
      Vector Selection Array=DataInterface() {propertyValues {FloatData [<8 -5 9>, <-3 -5 7>]}}
     )
    )) {isSetParameters=true}
   /SetVariables(Particles.PositionLightTest:NiagaraPosition=&Add_Vector(
     A=Linked() {linkedVariable {name=Particles.Position, type=NiagaraPosition}}
     B=&SelectVectorFromArray(
      Direct Array Index=&MakeIntFromFloat(Float=$Particles.TimeBlink)
      Vector Selection Array=DataInterface() {propertyValues {FloatData [<7 -3 6>, <-3 -3 6>]}}
     )
    )) {enabled=false, isSetParameters=true}
   &ScaleColor(Scale Alpha=&ClampFloat(Float=$User.Emissive), ScaleA=true, ScaleRGB=true)
  ]
  rendererClasses [&NiagaraLightRendererProperties, &NiagaraSpriteRendererProperties]
  renderers [
   {rendererClass=&NiagaraLightRendererProperties, rendererIndex=0}
   {rendererClass=&NiagaraSpriteRendererProperties, rendererIndex=1}
  ]
 }
 DirectionalBurst001 {
  emitterUpdateScript [
   &EmitterState()
   &SpawnBurst_Instantaneous(Use Spawn Probability=false, UseLoopCountLimit=true)
  ]
  particleSpawnScript [&InitializeParticle(
    Color=<rgba 4 4 4 0.1>
    Lifetime=6
    Mesh Scale Mode=::NewEnumerator3
    Sprite Size Mode=::NewEnumerator3
    Uniform Sprite Size=47.7503
    UsePositionOffset=false
    Write Material Random=true
   )]
  particleUpdateScript [
   &ParticleState(
    Kill Particles When Lifetime Has Elapsed=false
    Let Infinitely Lived Particles Die When Emitter Deactivates=true
    Loop Particles Lifetime=true
   )
   &SkeletalMeshLocation(
    MaskSpawnGroup=false
    Mesh Sampling Type=::NewEnumerator5
    Sampled Position Offset=<0 -1.3 0>
    Skeletal Mesh=$User.SkeletalMesh
    Socket Sampling Mode=::NewEnumerator5
   )
   &ScaleColor(Scale Alpha=&ClampFloat(Float=$User.Emissive), ScaleA=true, ScaleRGB=true)
   /SetVariables(Particles.Blink:NiagaraFloat=&FloatFromCurve(FloatCurve=DataInterface() {propertyValues {Curve {keys [
         {arriveTangent=-1.6748, interpMode=RCIM_Cubic, time=0, value=1}
         {arriveTangent=-0.0227, interpMode=RCIM_Cubic, leaveTangent=-0.0227, tangentMode=RCTM_User, time=0.1526, value=1.25}
         {arriveTangent=-0.209, interpMode=RCIM_Cubic, leaveTangent=-0.209, time=0.2819, value=1.1}
         {arriveTangent=-0.4585, interpMode=RCIM_Cubic, leaveTangent=-0.4585, time=0.3918, value=1.2}
         {arriveTangent=0.4333, interpMode=RCIM_Cubic, leaveTangent=0.4333, time=0.5, value=1}
         <at 0.738 1.35 RCIM_Cubic>
         <at 1 1 RCIM_Cubic>
        ]}}})) {isSetParameters=true}
   &FN_Light_Attributes(
    Color=$System.BaseLightColor
    Create Color Attribute=true
    Create Exponent Attribute=true
    Create Position Attribute=true
    Create Radius Attribute=true
    Enable Particle Light Enable=true
    Exponent=3
    Light Enabled:NiagaraBool=&InvertBool(Boolean:NiagaraBool=&Boolean_OrOperation(Bool1=$User.bIsFrontEnd, Bool2=$User.bIsFrontEndPreview))
    Position=<0 -3 0>
    Radius=30
    Scale Color=&Multiply_Float(A=$Particles.Blink, B=&ClampFloat(Float=$User.Emissive))
   )
  ]
  rendererClasses [
   &NiagaraMeshRendererProperties
   &NiagaraLightRendererProperties
   &NiagaraSpriteRendererProperties
  ]
  renderers [
   {rendererClass=&NiagaraMeshRendererProperties, rendererIndex=0}
   {rendererClass=&NiagaraLightRendererProperties, rendererIndex=1}
   {rendererClass=&NiagaraMeshRendererProperties, rendererIndex=2}
   {rendererClass=&NiagaraSpriteRendererProperties, rendererIndex=3}
  ]
 }
}
summary {
 systemName=NS_Backpack_LocalZilla
 emitters [
  {
   emitterName=DirectionalBurst
   rendererClasses [&NiagaraLightRendererProperties, &NiagaraSpriteRendererProperties]
  }
  {
   emitterName=DirectionalBurst001
   rendererClasses [
    &NiagaraMeshRendererProperties
    &NiagaraLightRendererProperties
    &NiagaraSpriteRendererProperties
   ]
  }
 ]
 userVariables [
  User.bIsFrontEnd {type=NiagaraBool}
  User.bIsFrontEndPreview {type=NiagaraBool}
  User.Emissive {type=NiagaraFloat, defaultValue=0, description=""}
  User.SkeletalMesh {
   type=NiagaraDataInterfaceSkeletalMesh
   defaultValue {
    struct=NiagaraExt_VariableValue_DataInterface
    value {dataInterface=&self:NiagaraDataInterfaceSkeletalMesh_0}
   }
  }
  User.VarientNight {type=NiagaraFloat, defaultValue=0, description=""}
 ]
}
